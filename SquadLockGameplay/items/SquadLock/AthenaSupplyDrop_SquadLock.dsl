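; AthenaSupplyDrop_SquadLock is a supply drop that carries a SquadLock child actor and configures how many players it takes to unlock.
; On BeginPlay it waits for playlist data to be ready, then reads the Row Name curve from AthenaGameData.
; If the row is found, the truncated value sets the lock's maximum, required and adjusted player counts.
; It also supplies its own SquadLock marker details ID.
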(event EventBeginPlay
  (|CallParentFunction)
  (Class|FortAsyncActionCurrentPlaylistReady|PlaylistDataReadyAsync
    (:OnReady
      (bind _assquad_lock (Utilities|Casting|CastToSquadLock (Variables|/Script/Engine.ChildActorComponent|GetChildActor (Variables|Default|GetSquadLock))))
      (bind _outxy (DataTable|EvaluateCurveTableRow "/Game/Athena/Balance/DataTables/AthenaGameData.AthenaGameData" (Variables|Default|GetRow%20Name))
        (:RowFound
          (Variables|/SquadLockGameplay/items/SquadLock/SquadLock.SquadLock_C|SetMaxRequiredPlayers (Math|Float|Truncate _outxy) _assquad_lock)
          (Variables|/SquadLockGameplay/items/SquadLock/SquadLock.SquadLock_C|SetNumPlayersRequiredToUnlock (Math|Float|Truncate _outxy) _assquad_lock)
          (Variables|/SquadLockGameplay/items/SquadLock/SquadLock.SquadLock_C|SetAdjustedNumPlayersRequiredToUnlock (Math|Float|Truncate _outxy) _assquad_lock))))))

(event Collision|EventActorBeginOverlap
  (params (OtherActor :type Actor%20Object%20Reference)))

(event EventTick
  (params (DeltaSeconds :type Float%20%28single-precision%29)))

(event MarkableActor|EventSetCanBeMarked
  (params (bInCanBeMarked :type Boolean)))

(fn GetMarkerDetailsId
  (return "MarkerDetails.SquadLock"))

(fn GetTargetTagsForQuestEvent
  (return 0))

(fn ConstructionScript
  (|CallParentFunction))
