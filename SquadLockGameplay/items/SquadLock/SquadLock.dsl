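; Squad lock Blueprint: an Athena scanner lock that tracks pawns in front of it and plays an unlock sequence.
; BeginPlay derives the required player count from playlist tags and squad size, with scalable-float overrides.
; Scanner-box overlaps update pawn counts and start scanning; a larger sphere applies an effect disabling teamup behavior.
; Timelines, timers and custom primitive data animate eye and rays; unlocking flashes, updates holo UI and sends quest events.
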(event Custom|OneTime_SetAdjustedNumPlayersRequiredToUnlock
  (params (PlaylistTags :type Gameplay%20Tag%20Container%20Structure) (FortPlaylist :type Fort%20Playlist%20Object%20Reference))
  (bind _scalablefloatfornumrequiredoverride_largeteam (Variables|Default|GetScalableFloatForNumRequiredOverride_LargeTeam))
  (bind _returnvalue (ScalableFloat|GetValueasInteger _scalablefloatfornumrequiredoverride_largeteam))
  (bind _scalablefloatfornumrequiredoverride_solos (Variables|Default|GetScalableFloatForNumRequiredOverride_Solos))
  (bind _returnvalue_2 (ScalableFloat|GetValueasInteger _scalablefloatfornumrequiredoverride_solos))
  (bind _scalablefloatfornumrequiredoverride_duos (Variables|Default|GetScalableFloatForNumRequiredOverride_Duos))
  (bind _returnvalue_3 (ScalableFloat|GetValueasInteger _scalablefloatfornumrequiredoverride_duos))
  (bind _scalablefloatfornumrequiredoverride_trios (Variables|Default|GetScalableFloatForNumRequiredOverride_Trios))
  (bind _returnvalue_4 (ScalableFloat|GetValueasInteger _scalablefloatfornumrequiredoverride_trios))
  (bind _scalablefloatfornumrequiredoverride_squads (Variables|Default|GetScalableFloatForNumRequiredOverride_Squads))
  (bind _returnvalue_5 (ScalableFloat|GetValueasInteger _scalablefloatfornumrequiredoverride_squads))
  (bind _maxsquadsize (Variables|/Script/FortniteGame.FortPlaylist|GetMaxSquadSize FortPlaylist))
  (bind _cvar_minrequiredplayers (Variables|Default|GetCVAR_MinRequiredPlayers))
  ;[1 1 1] Set Adjusted Num Required Players (which we'll rep and use from here on)
  (if (and (ScalableFloat|GetValueasBool (Variables|Default|GetScalableFloatUseSquadSizeForNumRequiredPlayers)) (Variables|Default|GetbAllowRequirePlayerCountOverrideFromPlaylist))
    (if (GameplayTags|HasAnyTags PlaylistTags (Variables|Default|GetPlaylistTag_Delulu) true)
      _
      (elif (GameplayTags|HasAnyTags PlaylistTags (Variables|Default|GetPlaylistTag_Solos) true)
        ;;[1 1 1] CheckPlaylistOverrides
        ;;;[1 1 1] Solos
        (if (>= _returnvalue_2 1)
          (Variables|Default|SetAdjustedNumPlayersRequiredToUnlock _returnvalue_2)
          (Fortnite|FortShippingLog 0 (Utilities|String|Append "SQUADLOCKS: Solos.  Setting num adjusted players to: " (Utilities|String|ToString(Integer) _returnvalue_2)))
          (else
            ;;[1 1 1] Set Adjusted required players from playlist squad size
            (if (< _maxsquadsize _cvar_minrequiredplayers) :!label _label
              (Variables|Default|SetMaxRequiredPlayers _cvar_minrequiredplayers)
              (else
                (Variables|Default|SetMaxRequiredPlayers _maxsquadsize :!label _label_2)))
            (bind _output_get (Variables|Default|SetAdjustedNumPlayersRequiredToUnlock (Variables|Default|GetMaxRequiredPlayers)))
            (Fortnite|FortShippingLog 0 (Utilities|String|Append "SQUADLOCKS: Setting num adjusted players to: " (Utilities|String|ToString(Integer) _output_get) "    Min Required Players:" (Utilities|String|ToString(Integer) _cvar_minrequiredplayers) "   Max Squad Size:" (Utilities|String|ToString(Integer) _maxsquadsize)))))
        (elif (GameplayTags|HasAnyTags PlaylistTags (Variables|Default|GetPlaylistTag_Duos) true)
          ;;;[1 1 1] Duos
          (if (>= _returnvalue_3 1)
            (Variables|Default|SetAdjustedNumPlayersRequiredToUnlock _returnvalue_3)
            (Fortnite|FortShippingLog 0 (Utilities|String|Append "SQUADLOCKS: Duos.  Setting num adjusted players to: " (Utilities|String|ToString(Integer) _returnvalue_3)))
            (else
              (goto _label)))
          (elif (GameplayTags|HasAnyTags PlaylistTags (Variables|Default|GetPlaylistTag_Trios) true)
            ;;;[1 1 1] Trios
            (if (>= _returnvalue_4 1)
              (Variables|Default|SetAdjustedNumPlayersRequiredToUnlock _returnvalue_4)
              (Fortnite|FortShippingLog 0 (Utilities|String|Append "SQUADLOCKS: Trios:  Setting num adjusted players to: " (Utilities|String|ToString(Integer) _returnvalue_4)))
              (else
                (goto _label)))
            (elif (GameplayTags|HasAnyTags PlaylistTags (Variables|Default|GetPlaylistTag_Squad) true)
              ;;;[1 1 1] Squads
              (if (>= _returnvalue_5 1)
                (Variables|Default|SetAdjustedNumPlayersRequiredToUnlock _returnvalue_5)
                (Fortnite|FortShippingLog 0 (Utilities|String|Append "SQUADLOCKS: Duos:  Setting num adjusted players to: " (Utilities|String|ToString(Integer) _returnvalue_5)))
                (else
                  (goto _label)))
              (elif (GameplayTags|HasAnyTags PlaylistTags (Variables|Default|GetPlaylistTag_LargeTeam) true)
                ;;;[1 1 1] LargeTeam
                (if (>= _returnvalue 1)
                  (Variables|Default|SetAdjustedNumPlayersRequiredToUnlock _returnvalue)
                  (Fortnite|FortShippingLog 0 (Utilities|String|Append "SQUADLOCKS: LargeTeam.  Setting num adjusted players to: " (Utilities|String|ToString(Integer) _returnvalue)))
                  (else
                    (goto _label)))
                (else
                  (goto _label))))))))
    (goto _label_2)
    (else
      (bind _output_get_2 (Variables|Default|SetAdjustedNumPlayersRequiredToUnlock (Variables|Default|GetNumPlayersRequiredToUnlock)))
      (Fortnite|FortShippingLog 0 (Utilities|String|Append "SQUADLOCKS: setting num adjusted players to: " (Utilities|String|ToString(Integer) _output_get_2) ".  ScalableFloatUseSquadSizeForNumRequiredPlayers was true")))))

(event Custom|TeamupVolumeEndOverlap
  (params (OverlappedComponent :type Primitive%20Component%20Object%20Reference) (OtherActor :type Actor%20Object%20Reference) (OtherComp :type Primitive%20Component%20Object%20Reference) (OtherBodyIndex :type Integer))
  ;[1 1 1] End Overlap on Larger Volume For Teamup  Tag
  (if (not (Variables|Default|GetHasBeenDistabled))
    (bind _asfort_pawn (Utilities|Casting|CastToFortPawn OtherActor)
      (:then
        (GameplayEffects|RemoveActiveEffectswithTags (Ability|GetAbilitySystemComponent _asfort_pawn) "(GameplayTags=((TagName=\"Athena.Gameplay.Status.TeamupBehaviorDisabled\")))")
        (Development|PrintString (Utilities|String|Append "Removing Teamup tag" (Utilities|GetDisplayName _asfort_pawn)) (Variables|Default|GetDrawDebuglinesToScannedPawns) (Variables|Default|GetDrawDebuglinesToScannedPawns))
        (Utilities|FlowControl|DoOnce false
          (:Completed
            (Fortnite|FortShippingLog 0 "SQUADLOCKS: First Pawn Ended Overlap with Larger Volume"))))
      (:CastFailed
        (Utilities|FlowControl|DoOnce false
          (:Completed
            (Fortnite|FortShippingLog 0 (Utilities|String|Append "SQUADLOCKS EndOverlap LargerVolume Failed to cast fort pawn from actor: " (Utilities|GetDisplayName OtherActor)))))))))

(event Custom|ReinitializeCollisionSphereForDisableTeamupBehaviorTag
  ;[1 1 1] Start Larger Volume For Disabling Teamup Tag
  (Utilities|IsValid (Variables|Default|GetCollisionSphereForDisableTeamupBehaviorTag)
    (:Is%20Valid
      (Fortnite|FortShippingLog 0 "SQUADLOCKS: Binding to Larger Volume Overlap")
      (Collision|BindEventtoOnComponentBeginOverlap (Variables|Default|GetCollisionSphereForDisableTeamupBehaviorTag) (delegate Custom|TeamupVolumeStartOverlap))
      ;[1 1 1] End Overlap on Larger Volume For Teamup  Tag
      (Collision|BindEventtoOnComponentEndOverlap (Variables|Default|GetCollisionSphereForDisableTeamupBehaviorTag) (delegate Custom|TeamupVolumeEndOverlap)))
    (:Is%20Not%20Valid
      (Utilities|Time|SetTimerbyEvent (delegate Custom|ReinitializeCollisionSphereForDisableTeamupBehaviorTag) 0.1)
      (Utilities|FlowControl|DoOnce false
        (:Completed
          (Fortnite|FortShippingLog 0 (Utilities|String|Append "SQUADLOCKS: CollisionSphereForDisableTeamupBehaviorTag is invalid startin timer to for later initialization.  Actor paht = " (Utilities|GetObjectPathString self))))))))

(event Custom|TeamupVolumeStartOverlap
  (params (OverlappedComponent :type Primitive%20Component%20Object%20Reference) (OtherActor :type Actor%20Object%20Reference) (OtherComp :type Primitive%20Component%20Object%20Reference) (OtherBodyIndex :type Integer) (bFromSweep :type Boolean) (SweepResult :type Hit%20Result%20Structure%20%28by%20ref%29))
  ;[1 1 1] Start Larger Volume For Disabling Teamup Tag
  (if (not (Variables|Default|GetHasBeenDistabled))
    (bind _asfort_pawn (Utilities|Casting|CastToFortPawn OtherActor)
      (:then
        (Utilities|IsValid (Ability|GetAbilitySystemComponent _asfort_pawn)
          (:Is%20Valid
            (GameplayEffects|ApplyGameplayEffectToSelf :self (Ability|GetAbilitySystemComponent _asfort_pawn) :GameplayEffectClass (Variables|Default|GetGameplay%20Effect_Apply_AC_DisableTeamupCheckTag) :Level 0.0)
            (Development|PrintString (Utilities|String|Append "Adding Teamup tag" (Utilities|GetDisplayName _asfort_pawn)) (Variables|Default|GetDrawDebuglinesToScannedPawns) (Variables|Default|GetDrawDebuglinesToScannedPawns))
            (Utilities|FlowControl|DoOnce false
              (:Completed
                (Fortnite|FortShippingLog 0 "SQUADLOCKS: First Pawn Overlapped Larger Volume"))))))
      (:CastFailed
        (Utilities|FlowControl|DoOnce false
          (:Completed
            (Fortnite|FortShippingLog 0 (Utilities|String|Append "SQUADLOCKS: Larger Volume Start Overlap failed to cast fort pawn from actor: " (Utilities|GetDisplayName OtherActor)))))))))

(event EventBeginPlay
  ;[1 1 1] Begin Play
  (|CallParentFunction)
  (Class|FortAsyncActionCurrentPlaylistReady|PlaylistDataReadyAsync
    (:OnReady
      (bind _newparam (CallFunction|GetScannerForwardRot))
      (bind _output_get (Variables|ScannerRays|SetStartingtRot _newparam))
      (bind _output_get_2 (Variables|ScannerRays|SetCurrentRot _output_get))
      (Variables|ScannerRays|SetPrimaryDesiredRot _output_get_2)
      (Variables|ScannerRays|SetAimInterpolationAlpha 0.0)
      (Utilities|Array|AddUnique (Variables|Default|GetScannerRays) (Utilities|Struct|MakeSquadLockScannerRay 0 (Variables|Default|GetSm_Camera_ScannerRay_ForPawns)))
      (Utilities|Array|AddUnique (Variables|Default|GetScannerRays) (Utilities|Struct|MakeSquadLockScannerRay 0 (Variables|Default|GetSm_Camera_ScannerRay_Volume)))
      (Utilities|Array|AddUnique (Variables|Default|GetScannerRays) (Utilities|Struct|MakeSquadLockScannerRay 0 (Variables|Default|GetSm_Camera_ScannerRay3)))
      (Utilities|Array|AddUnique (Variables|Default|GetScannerRays) (Utilities|Struct|MakeSquadLockScannerRay 0 (Variables|Default|GetSm_Camera_ScannerRay4)))
      (Networking|SwitchHasAuthority
        (:Authority
          (Utilities|FlowControl|Sequence
            (:then_0
              ;[1 1 1] Start Scanner Volume Overlap
              (Utilities|FlowControl|Sequence
                (:then_0
                  ;[1 1 1] Start Larger Volume For Disabling Teamup Tag
                  (Utilities|IsValid (Variables|Default|GetCollisionSphereForDisableTeamupBehaviorTag)
                    (:Is%20Valid
                      (Fortnite|FortShippingLog 0 "SQUADLOCKS: Binding to Larger Volume Overlap")
                      (Collision|BindEventtoOnComponentBeginOverlap (Variables|Default|GetCollisionSphereForDisableTeamupBehaviorTag) (delegate Custom|TeamupVolumeStartOverlap))
                      ;[1 1 1] End Overlap on Larger Volume For Teamup  Tag
                      (Collision|BindEventtoOnComponentEndOverlap (Variables|Default|GetCollisionSphereForDisableTeamupBehaviorTag) (delegate Custom|TeamupVolumeEndOverlap)))
                    (:Is%20Not%20Valid
                      (Utilities|Time|SetTimerbyEvent (delegate Custom|ReinitializeCollisionSphereForDisableTeamupBehaviorTag) 0.1)
                      (Utilities|FlowControl|DoOnce false
                        (:Completed
                          (Fortnite|FortShippingLog 0 (Utilities|String|Append "SQUADLOCKS: CollisionSphereForDisableTeamupBehaviorTag is invalid startin timer to for later initialization.  Actor paht = " (Utilities|GetObjectPathString self))))))))
                (:then_1
                  (Utilities|IsValid (Variables|Default|GetCollisionBoxForPlayerStandingInFrontOfScanner)
                    (:Is%20Valid
                      (Fortnite|FortShippingLog 0 "SQUADLOCKS: Binding to Scanner Volume Overlap")
                      (Collision|BindEventtoOnComponentBeginOverlap (Variables|Default|GetCollisionBoxForPlayerStandingInFrontOfScanner) (delegate Custom|Scanner%20Collis%20Overlap))
                      ;[1 1 1] End Scanner Volume Overlap
                      (Collision|BindEventtoOnComponentEndOverlap (Variables|Default|GetCollisionBoxForPlayerStandingInFrontOfScanner) (delegate Custom|Scanner%20Collis%20End%20Overlap)))
                    (:Is%20Not%20Valid
                      (Utilities|Time|SetTimerbyEvent (delegate Custom|ReinitializeCollisionBoxForPlayerStandingInFrontOfScanner) 0.1)
                      (Utilities|FlowControl|DoOnce false
                        (:Completed
                          (Fortnite|FortShippingLog 0 (Utilities|String|Append "SQUADLOCKS: CollisionBoxForPlayerStandingInFrontOfScanner is invalid startin timer to for later initialization.  Actor paht = " (Utilities|GetObjectPathString self))))))))))
            (:then_1
              (bind (_gamestate _playlist _playlistcontexttags) (Class|FortAsyncActionCurrentPlaylistReady|PlaylistDataReadyAsync)
                (:OnReady
                  (if (ScalableFloat|GetValueasBool (Variables|Default|GetScalableFloatFor_Enabled))
                    (bind _asfort_game_mode_athena (Utilities|Casting|CastToFortGameModeAthena (Game|GetGameMode))
                      (:then
                        (Variables|Default|SetAs%20Game%20Mode%20Athena _asfort_game_mode_athena)
                        (CallFunction|OneTimeSetAdjustedNumPlayersRequiredtoUnlock _playlistcontexttags _playlist))
                      (:CastFailed
                        (Fortnite|FortShippingLog 0 "SQUADLOCKS: Cast to FortGameModeAthena Failed")))
                    (else
                      (Variables|Default|SetHasBeenDistabled true)
                      (Fortnite|FortShippingLog 0 "SQUADLOCKS: have been disabled"))))))))
        (:Remote
          (CallFunction|InitLockScreenWidget)
          (CallFunction|OnRepBIsUnlocked))))))

(event Custom|Scanner%20Collis%20End%20Overlap
  (params (OverlappedComponent :type Primitive%20Component%20Object%20Reference) (OtherActor :type Actor%20Object%20Reference) (OtherComp :type Primitive%20Component%20Object%20Reference) (OtherBodyIndex :type Integer))
  ;[1 1 1] End Scanner Volume Overlap
  (if (not (Variables|Default|GetHasBeenDistabled))
    (if (CallFunction|IsLocked)
      (bind _asfort_pawn (Utilities|Casting|CastToFortPawn OtherActor)
        (:then
          (Utilities|Casting|CastToFortPlayerStateAthena (Variables|/Script/Engine.Pawn|GetPlayerState _asfort_pawn)
            (:then
              _)
            (:CastFailed
              _))
          (Utilities|Array|RemoveItem (Variables|Default|GetAs%20Fort%20Pawn) _asfort_pawn)
          (Development|PrintString (Utilities|String|Append "Removing pawn from AsFortPawns: New Size:" (Utilities|String|ToString(Integer) (Utilities|Array|Length (Variables|Default|GetAs%20Fort%20Pawn)))) (Variables|Default|GetDrawDebuglinesToScannedPawns) (Variables|Default|GetDrawDebuglinesToScannedPawns))
          (CallFunction|UpdatePawnCounts false)
          (Utilities|FlowControl|DoOnce false
            (:Completed
              (Fortnite|FortShippingLog 0 "SQUADLOCKS: First Pawn Ended Scanner Overlap"))))
        (:CastFailed
          (Utilities|FlowControl|DoOnce false
            (:Completed
              (Fortnite|FortShippingLog 0 (Utilities|String|Append "SQUADLOCKS: End Overlap Scanner failed to cast fort pawn from actor: " (Utilities|GetDisplayName OtherActor)))))))
      (else
        (Utilities|FlowControl|DoOnce false
          (:Completed
            (Fortnite|FortShippingLog 0 "SQUADLOCKS: Scanner End Overlap:is Locked")))))))

(event Custom|Scanner%20Collis%20Overlap
  (params (OverlappedComponent :type Primitive%20Component%20Object%20Reference) (OtherActor :type Actor%20Object%20Reference) (OtherComp :type Primitive%20Component%20Object%20Reference) (OtherBodyIndex :type Integer) (bFromSweep :type Boolean) (SweepResult :type Hit%20Result%20Structure%20%28by%20ref%29))
  ;[1 1 1] Start Scanner Volume Overlap
  (if (not (Variables|Default|GetHasBeenDistabled))
    (if (CallFunction|IsLocked)
      (bind _asfort_pawn (Utilities|Casting|CastToFortPawn OtherActor)
        (:then
          (Utilities|Array|AddUnique (Variables|Default|GetAs%20Fort%20Pawn) _asfort_pawn)
          (Development|PrintString (Utilities|String|Append "Adding pawn to AsFortPawns: New Size:" (Utilities|String|ToString(Integer) (Utilities|Array|Length (Variables|Default|GetAs%20Fort%20Pawn)))) (Variables|Default|GetDrawDebuglinesToScannedPawns) (Variables|Default|GetDrawDebuglinesToScannedPawns))
          (CallFunction|UpdatePawnCounts false)
          (CallFunction|StartScanningAll)
          (Utilities|FlowControl|DoOnce false
            (:Completed
              (Fortnite|FortShippingLog 0 "SQUADLOCKS: First Pawn Overlapped"))))
        (:CastFailed
          (Utilities|FlowControl|DoOnce false
            (:Completed
              (Fortnite|FortShippingLog 0 (Utilities|String|Append "SQUADLOCKS: ScannerVolume start overlap failed to cast to fortPawn from actor: " (Utilities|GetDisplayName OtherActor)))))))
      (else
        (Utilities|FlowControl|DoOnce false
          (:Completed
            (Fortnite|FortShippingLog 0 "SQUADLOCKS: SquadLock is not locked")))))))

(event Custom|ReinitializeCollisionBoxForPlayerStandingInFrontOfScanner
  ;[1 1 1] Start Scanner Volume Overlap
  (Utilities|IsValid (Variables|Default|GetCollisionBoxForPlayerStandingInFrontOfScanner)
    (:Is%20Valid
      (Fortnite|FortShippingLog 0 "SQUADLOCKS: Binding to Scanner Volume Overlap")
      (Collision|BindEventtoOnComponentBeginOverlap (Variables|Default|GetCollisionBoxForPlayerStandingInFrontOfScanner) (delegate Custom|Scanner%20Collis%20Overlap))
      ;[1 1 1] End Scanner Volume Overlap
      (Collision|BindEventtoOnComponentEndOverlap (Variables|Default|GetCollisionBoxForPlayerStandingInFrontOfScanner) (delegate Custom|Scanner%20Collis%20End%20Overlap)))
    (:Is%20Not%20Valid
      (Utilities|Time|SetTimerbyEvent (delegate Custom|ReinitializeCollisionBoxForPlayerStandingInFrontOfScanner) 0.1)
      (Utilities|FlowControl|DoOnce false
        (:Completed
          (Fortnite|FortShippingLog 0 (Utilities|String|Append "SQUADLOCKS: CollisionBoxForPlayerStandingInFrontOfScanner is invalid startin timer to for later initialization.  Actor paht = " (Utilities|GetObjectPathString self))))))))

(event Interaction|EventBlueprintOnInteract
  (params (InteractingPawn :type Fort%20Pawn%20Object%20Reference) (InteractionBeingAttempted :type EInteractionBeingAttempted%20Enum :default "FirstInteraction"))
  ;[1 1 1] On Unlocked/Pending unlocked
  ;;[1 1 1] Comment
  (Development|PrintString "On Interact" (Variables|Default|GetDrawDebuglinesToScannedPawns) (Variables|Default|GetDrawDebuglinesToScannedPawns)))

(event EventOnUnlocked
  (bind _currentpawninvolumeindex (Variables|Default|GetCurrentPawnInVolumeIndex))
  (bind _pawnsinvolume (Variables|Default|GetPawnsInVolume))
  (bind (_output_pawn_17_ca2911e448c8191bd27fa28c945c96fa _output_niagarasystem_27_9161e59c4f3629e93c3b238aa11cb77b) (Utilities|Struct|BreakPawnsInSquadLockVolume (Utilities|Array|Get(aref) _pawnsinvolume _currentpawninvolumeindex)))
  ;[1 1 1] On Unlocked/Pending unlocked
  (Utilities|FlowControl|Sequence
    (:then_0
      (|CallParentFunction)
      (Variables|Default|SetbIsUnlocked true)
      (CallFunction|SendQuestUpdateforPlayers)
      (Rendering|Material|SetMaterial (Variables|Default|GetLock) 0 "/SquadLockGameplay/items/SquadLock/Assets/Materials/MI_Squad_Lock_Green.MI_Squad_Lock_Green")
      (bind _audio (Variables|Default|GetAudio)
        (:then
          (Audio|SpawnSoundAttached (Variables|Unlock|GetUnlockedSound) _audio)
          ;;[1 1 1] Lock unlocked burst effect
          (Utilities|FlowControl|Delay 1.5)
          (Niagara|SpawnSystematLocation :SystemTemplate "/SquadLockGameplay/items/SquadLock/Assets/FX/NS_SquadLock_VaultOpen.NS_SquadLock_VaultOpen" :Location (+ (Transformation|GetActorLocation) (Math|Vector|RotateVector "0.000000,-20.000000, 0" (Transformation|GetActorRotation))) :Scale "1.500000,1.500000,1.500000" :bAutoDestroy true :bAutoActivate true :PoolingMethod "None" :bPreCullCheck true)
          ;;[1 1 1] Deactivate Niagara systems
          (Niagara|SetNiagaraVariable(Float) _output_niagarasystem_27_9161e59c4f3629e93c3b238aa11cb77b "AgeInput" 1.0) ; ideally this would run off a timer
          (Components|Activation|Deactivate _output_niagarasystem_27_9161e59c4f3629e93c3b238aa11cb77b))
        (:else)))
    (:then_1
      (Utilities|FlowControl|Delay 4.0)
      (Rendering|Material|SetMaterial (Variables|Default|GetLock) 0 "/SquadLockGameplay/items/SquadLock/Assets/Materials/MI_Squad_Lock_Black.MI_Squad_Lock_Black")
      (Audio|PlaySoundatLocation (Variables|Default|GetScannerOffSound) (Transformation|GetWorldLocation (Variables|Default|GetAudio)))
      ;;[1 1 1] [AI] Destroy SmartObject Component when the SquadLock is unlocked
      (Networking|SwitchHasAuthority
        (:Authority
          (Ability|GameplayEffect|AddLooseGameplayTags self (Variables|Default|GetUnlockStatusGameplayTags)))))))

(event Custom|StartUnlock
  ;[1 1 1] On Unlocked/Pending unlocked
  (Utilities|FlowControl|DoOnce false
    (:Completed
      (bind _newparam (CallFunction|GetScannerForwardRot))
      (CallFunction|SetupandStartScannerAimInterpolation _newparam (Variables|ScannerRays|GetPawnEnterScannerColor))
      (CallFunction|FlashScanner))))

(event Custom|FlashScanner
  ;[1 1 1] On Unlocked/Pending unlocked
  ;;[1 1 1] Flash
  (CallFunction|SetScannerPower false)
  (Utilities|FlowControl|Delay (Variables|Default|GetUnlockFlashOffDelaySeconds))
  (CallFunction|SetScannerPower true)
  (Utilities|FlowControl|Sequence
    (:then_0
      _)
    (:then_1
      (Utilities|FlowControl|DoOnce false
        (:Completed
          (Audio|PlaySoundatLocation (Variables|Default|GetPendingUnlockFlashSound) (Transformation|GetWorldLocation (Variables|Default|GetAudio)))
          (Math|Integer|IncrementInt (Variables|Default|GetFlashBeepCount)
            (:%20%20
              _))))))
  (Utilities|FlowControl|Delay (Variables|Unlock|GetUnlockFlashOnDelaySeconds))
  (bind _result (Math|Integer|IncrementInt (Variables|Default|GetFlashCount))
    (:%20%20
      (if (>= _result 3)
        (CallFunction|SetScannerPower false)
        (Audio|PlaySoundatLocation (Variables|Default|GetFinalLockVisualFlashSound) (Transformation|GetActorLocation))
        ;;;[1 1 1] Put locked padlock icon on our UI display
        (Class|SquadLockHoloUI|ShowPadlock (Variables|Default|GetLockScreenUI))
        (Utilities|FlowControl|Delay 2.0)
        ;;;[1 1 1] Unlock sequence in UI fires from lock state notify
        (Variables|LockParent|SetLockState "NewEnumerator2")
        ;;;[1 1 1] FadeOutAudio
        (Audio|Components|Audio|FadeOut (Variables|Default|GetScanSound) 0.05 0.0 "Linear" :!class /Script/Engine.AudioComponent)
        (bind _audio (Variables|Default|GetAudio)
          (:then
            (Audio|Components|Audio|FadeOut _audio 0.3 0.0 "Linear" :!class /Script/Engine.AudioComponent)
            (SoundIndicator|StopSoundIndicatorLooping (Variables|Default|GetFortSoundIndicator))
            (CallFunction|ResetDoOnce))
          (:else))
        (else
          (CallFunction|FlashScanner))))))

(event Custom|ResetDoOnce
  ;[1 1 1] On Unlocked/Pending unlocked
  ;;[1 1 1] Flash
  (Utilities|FlowControl|DoOnce false
    (:Completed
      (Audio|PlaySoundatLocation (Variables|Default|GetPendingUnlockFlashSound) (Transformation|GetWorldLocation (Variables|Default|GetAudio)))
      (Math|Integer|IncrementInt (Variables|Default|GetFlashBeepCount)
        (:%20%20
          (Utilities|FlowControl|Delay (Variables|Unlock|GetUnlockFlashOnDelaySeconds))
          (bind _result (Math|Integer|IncrementInt (Variables|Default|GetFlashCount))
            (:%20%20
              (if (>= _result 3)
                (CallFunction|SetScannerPower false)
                (Audio|PlaySoundatLocation (Variables|Default|GetFinalLockVisualFlashSound) (Transformation|GetActorLocation))
                ;;;[1 1 1] Put locked padlock icon on our UI display
                (Class|SquadLockHoloUI|ShowPadlock (Variables|Default|GetLockScreenUI))
                (Utilities|FlowControl|Delay 2.0)
                ;;;[1 1 1] Unlock sequence in UI fires from lock state notify
                (Variables|LockParent|SetLockState "NewEnumerator2")
                ;;;[1 1 1] FadeOutAudio
                (Audio|Components|Audio|FadeOut (Variables|Default|GetScanSound) 0.05 0.0 "Linear" :!class /Script/Engine.AudioComponent)
                (bind _audio (Variables|Default|GetAudio)
                  (:then
                    (Audio|Components|Audio|FadeOut _audio 0.3 0.0 "Linear" :!class /Script/Engine.AudioComponent)
                    (SoundIndicator|StopSoundIndicatorLooping (Variables|Default|GetFortSoundIndicator))
                    (CallFunction|ResetDoOnce))
                  (:else))
                (else
                  (CallFunction|FlashScanner))))))))))

(event Custom|SendQuestUpdateForPlayers
  (bind _as_fort_pawn (Variables|Default|GetAs%20Fort%20Pawn))
  ;[1 1 1] On Unlocked/Pending unlocked
  ;;[1 1 1] When unlocked send a quest stat event to each player's quest manager
  (if (Utilities|Array|IsNotEmpty _as_fort_pawn)
    (for _array_element _as_fort_pawn
      (Utilities|IsValid _array_element
        (:Is%20Valid
          (CustomVerb|BroadcastCustomVerbMessage (CustomVerb|MakeCustomVerbMessage _array_element (Variables|Default|GetUnlockCustomVerbName)))
          (Utilities|IsValid (Fortnite|GetFortPlayerControllerfromActor _array_element)
            (:Is%20Valid
              (if (Utilities|IsValid (Quests|GetQuestManager (Fortnite|GetFortPlayerControllerfromActor _array_element) "Invalid") :!class /Script/Engine.KismetSystemLibrary)
                (Quest|SendComplexCustomStatEvent (Quests|GetQuestManager (Fortnite|GetFortPlayerControllerfromActor _array_element) "Invalid") 0 (Variables|Quest|GetUnlockQuestSourceTag) "(GameplayTags=)" 1)))))))))

(event Custom|CE_AimEye
  (params (Interacting%20Pawn :type Fort%20Pawn%20Object%20Reference) (Index :type Integer))
  (bind _scannerrayindex (Variables|Default|GetScannerRayIndex))
  (bind _scannerrays (Variables|Default|GetScannerRays))
  (bind (_output_pawn_17_ca2911e448c8191bd27fa28c945c96fa _output_smcomp_2_9161e59c4f3629e93c3b238aa11cb77b) (Utilities|Struct|BreakSquadLockScannerRay (Utilities|Array|Get(aref) _scannerrays _scannerrayindex)))
  (bind _returnvalue (Transformation|GetWorldLocation _output_smcomp_2_9161e59c4f3629e93c3b238aa11cb77b))
  (bind _returnvalue_2 (Transformation|GetActorLocation Interacting%20Pawn))
  (bind _returnvalue_3 (+ _returnvalue_2 "0, 0,70.000000"))
  (bind _returnvalue_4 (Math|Rotator|FindLookatRotation _returnvalue _returnvalue_3))
  (bind _scannerrayindex_2 (Variables|Default|GetScannerRayIndex))
  (bind _scannerrays_2 (Variables|Default|GetScannerRays))
  (bind (_output_pawn_17_ca2911e448c8191bd27fa28c945c96fa_2 _output_smcomp_2_9161e59c4f3629e93c3b238aa11cb77b_2 _output_aimactive_10_7ce804af4b674e9fe898fcbb9eda2c90) (Utilities|Struct|BreakSquadLockScannerRay (Utilities|Array|Get(aref) _scannerrays_2 _scannerrayindex_2)))
  (bind (_output_pawn_17_ca2911e448c8191bd27fa28c945c96fa_3 _output_smcomp_2_9161e59c4f3629e93c3b238aa11cb77b_3) (Utilities|Struct|BreakSquadLockScannerRay (Utilities|Array|Get(aref) (Variables|Default|GetScannerRays) (Variables|Default|GetScannerRayIndex))))
  ;[0 0.1485 1] Aim The "Eye"
  (Utilities|FlowControl|Sequence
    (:then_0
      (Variables|Default|SetScannerRayIndex Index)
      (if (and (Utilities|Array|IsValidIndex (Variables|Default|GetScannerRays) (Variables|Default|GetScannerRayIndex)) (not _output_aimactive_10_7ce804af4b674e9fe898fcbb9eda2c90))
        (Utilities|Array|SetArrayElem _scannerrays_2 (Variables|Default|GetScannerRayIndex) (Utilities|Struct|MakeSquadLockScannerRay Interacting%20Pawn _output_smcomp_2_9161e59c4f3629e93c3b238aa11cb77b_2 true) false)
        (CallFunction|SetScanningActive true)
        (Rendering|SetVisibility _output_smcomp_2_9161e59c4f3629e93c3b238aa11cb77b true false)
        (if (== 0 Index)
          (CallFunction|SetupandStartScannerAimInterpolation _returnvalue_4 (Variables|ScannerRays|GetPawnEnterScannerColor))
          (CallFunction|SetScannerColor (Variables|Default|GetSm_Camera_ScannerRay_Volume) (Variables|ScannerRays|GetPawnEnterScannerColor))
          (else
            (Transformation|SetWorldRotation _output_smcomp_2_9161e59c4f3629e93c3b238aa11cb77b_3 _returnvalue_4 false false)))
        (if (Networking|IsServer)
          (CallFunction|CountActiveScanners))))
    (:then_1
      (Utilities|Time|SetTimerbyEvent (delegate Custom|DoneWithScan) (Variables|Default|GetScanTime)))))

(event Custom|DoneWithScan
  ;[0 0.1485 1] Aim The "Eye"
  (bind (_output_pawn_17_ca2911e448c8191bd27fa28c945c96fa _output_smcomp_2_9161e59c4f3629e93c3b238aa11cb77b _output_aimactive_10_7ce804af4b674e9fe898fcbb9eda2c90 _output_unaimactive_12_9b66154e49cf5dd3a1cabab49f3169ac) (Utilities|Struct|BreakSquadLockScannerRay (Utilities|Array|Get(aref) (Variables|Default|GetScannerRays) (stale Custom|CE_AimEye Index))))
  (Utilities|Array|SetArrayElem (Variables|Default|GetScannerRays) (stale Custom|CE_AimEye Index) (Utilities|Struct|MakeSquadLockScannerRay _output_pawn_17_ca2911e448c8191bd27fa28c945c96fa _output_smcomp_2_9161e59c4f3629e93c3b238aa11cb77b false _output_unaimactive_12_9b66154e49cf5dd3a1cabab49f3169ac) false)
  (CallFunction|UpdateScanningQues))

(event Custom|FinishedAllEyeAims
  ;[0.5156 0 0.1451] When all players have exited the volume (player count has reached zero)
  (Variables|Default|SetDisablingScanner true)
  (bind _returnvalue (Utilities|Time|SetTimerbyEvent (delegate Custom|UnaimEyeAll) 1.25))
  (Variables|Default|SetTimer_UnAimEye _returnvalue))

(event Custom|UnaimEyeAll
  (bind (_output_pawn_17_ca2911e448c8191bd27fa28c945c96fa _output_smcomp_2_9161e59c4f3629e93c3b238aa11cb77b) (Utilities|Struct|BreakSquadLockScannerRay (Utilities|Array|Get(aref) (Variables|Default|GetScannerRays) 1)))
  (bind _aim_eye (Variables|Components|GetAim%20eye))
  (bind _lock (Variables|Default|GetLock))
  ;[0.5156 0 0.1451] When all players have exited the volume (player count has reached zero)
  (Development|SetHiddeninGame (Variables|Default|GetSpotLight) true false)
  (Rendering|Material|SetCustomPrimitiveDataVector3 _lock 14 "0, 0, 0")
  ;;[1 1 1] Stop the aim timer if it's playing
  (if (Components|Timeline|IsPlaying _aim_eye)
    (Components|Timeline|Stop _aim_eye))
  (bind (_direction _eyelerp _stareintent _objectscale _adjusteyerings) (Timeline|UnAim%20Eye Play)
    (:Update
      (Rendering|Material|SetCustomPrimitiveDataFloat _lock 12 _eyelerp)
      (Rendering|Material|SetCustomPrimitiveDataFloat _lock 13 _stareintent)
      (Rendering|Material|SetCustomPrimitiveDataFloat _lock 26 (.x _adjusteyerings))
      (Rendering|Material|SetCustomPrimitiveDataFloat _lock 27 (.y _adjusteyerings))
      (for _array_element (Variables|Default|GetScannerRays)
        (bind (_array_element_pawn_17_ca2911e448c8191bd27fa28c945c96fa _array_element_smcomp_2_9161e59c4f3629e93c3b238aa11cb77b _array_element_aimactive_10_7ce804af4b674e9fe898fcbb9eda2c90 _array_element_unaimactive_12_9b66154e49cf5dd3a1cabab49f3169ac) (Utilities|Struct|BreakSquadLockScannerRay _array_element))
        (if _array_element_unaimactive_12_9b66154e49cf5dd3a1cabab49f3169ac
          (Transformation|SetRelativeScale3D _array_element_smcomp_2_9161e59c4f3629e93c3b238aa11cb77b (* (Math|Conversions|ToVector(Float) _objectscale) "0, 0, 0")))))
    (:Finished
      (Rendering|SetVisibility _output_smcomp_2_9161e59c4f3629e93c3b238aa11cb77b false false)
      (CallFunction|SetScanningActive false)
      ;;[0.4271 0.4271 0.4271] Deactivate Scan Volume
      (CallFunction|DeactivateScanVolumeNS)
      (Audio|PlaySoundatLocation (Variables|Default|GetScannerOffSound) (Transformation|GetActorLocation)))))

(event Custom|CE_EndAimEye
  (params (Index :type Integer) (InFortPawn :type Fort%20Pawn%20Object%20Reference))
  (bind (_output_pawn_17_ca2911e448c8191bd27fa28c945c96fa _output_smcomp_2_9161e59c4f3629e93c3b238aa11cb77b) (Utilities|Struct|BreakSquadLockScannerRay (Utilities|Array|Get(aref) (Variables|Default|GetScannerRays) (Variables|Default|GetScannerRayIndex))))
  (bind (_output_pawn_17_ca2911e448c8191bd27fa28c945c96fa_2 _output_smcomp_2_9161e59c4f3629e93c3b238aa11cb77b_2) (Utilities|Struct|BreakSquadLockScannerRay (Utilities|Array|Get(aref) (Variables|Default|GetScannerRays) (Variables|Default|GetScannerRayIndex))))
  (bind (_output_pawn_17_ca2911e448c8191bd27fa28c945c96fa_3 _output_smcomp_2_9161e59c4f3629e93c3b238aa11cb77b_3) (Utilities|Struct|BreakSquadLockScannerRay (Utilities|Array|Get(aref) (Variables|Default|GetScannerRays))))
  ;[1 0 0.7887] End aiming of the "Eye"
  (Variables|Default|SetScannerRayIndex Index)
  (if (Utilities|Array|IsValidIndex (Variables|Default|GetScannerRays) (Variables|Default|GetScannerRayIndex))
    (Utilities|FlowControl|Sequence
      (:then_0
        (if (== (Variables|Default|GetScannerRayIndex) 0)
          (CallFunction|SetupandStartScannerAimInterpolation (Math|Rotator|FindLookatRotation (Transformation|GetWorldLocation _output_smcomp_2_9161e59c4f3629e93c3b238aa11cb77b_3) (+ (Transformation|GetActorLocation InFortPawn) "0,0.000000,70.000000")) (Variables|ScannerRays|GetPawnExitScannerColor) true)
          (CallFunction|SetScannerColor (Variables|Default|GetSm_Camera_ScannerRay_Volume) (Variables|ScannerRays|GetPawnExitScannerColor))
          (CallFunction|SetPrimaryAimtoCurrent)
          ;[1 0 0.03895] spawn red fx on pawns on exit
          (Niagara|SpawnSystemAttached "/SquadLockGameplay/items/SquadLock/Assets/FX/NS_SquadLock_ExitVolume.NS_SquadLock_ExitVolume" (Variables|/Script/Engine.Character|GetMesh InFortPawn) "Head" (+ (Transformation|GetWorldLocation (Variables|/Script/Engine.Character|GetMesh InFortPawn)) "0,0,100.000000") "0, 0, 0" "KeepWorldPosition")
          (else
            (Rendering|SetVisibility _output_smcomp_2_9161e59c4f3629e93c3b238aa11cb77b_2 false false))))
      (:then_1
        (Utilities|Array|SetArrayElem (Variables|Default|GetScannerRays) (Variables|Default|GetScannerRayIndex) (Utilities|Struct|MakeSquadLockScannerRay InFortPawn _output_smcomp_2_9161e59c4f3629e93c3b238aa11cb77b false true) false))
      (:then_2
        (Utilities|Time|SetTimerbyEvent (delegate Custom|ExitScanTimerExpired) (Variables|Default|GetExitScanTime))))))

(event Custom|Deactivate%20Scan%20Volume%20NS
  ;[0.5156 0 0.1451] When all players have exited the volume (player count has reached zero)
  ;;[0.4271 0.4271 0.4271] Deactivate Scan Volume
  (Niagara|SetNiagaraVariable(Float) (Variables|Default|GetNS_SquadLock_ScanVolume) "AgeInput" 1.0))

(event Custom|ExitScanTimerExpired
  (bind _scannerrays (Variables|Default|GetScannerRays))
  ;[1 0 0.7887] End aiming of the "Eye"
  (bind (_output_pawn_17_ca2911e448c8191bd27fa28c945c96fa _output_smcomp_2_9161e59c4f3629e93c3b238aa11cb77b _output_aimactive_10_7ce804af4b674e9fe898fcbb9eda2c90) (Utilities|Struct|BreakSquadLockScannerRay (Utilities|Array|Get(aref) _scannerrays (stale Custom|CE_EndAimEye Index))))
  (Rendering|SetVisibility _output_smcomp_2_9161e59c4f3629e93c3b238aa11cb77b false false)
  (Utilities|Array|SetArrayElem (Variables|Default|GetScannerRays) (stale Custom|CE_EndAimEye Index) (Utilities|Struct|MakeSquadLockScannerRay _output_pawn_17_ca2911e448c8191bd27fa28c945c96fa _output_smcomp_2_9161e59c4f3629e93c3b238aa11cb77b _output_aimactive_10_7ce804af4b674e9fe898fcbb9eda2c90) false)
  (CallFunction|UpdateScanningQues))

(event Custom|OnRep_bIsUnlocked
  (bind _lockscreenwidget (Variables|Default|GetLockScreenWidget))
  (bind _returnvalue (UserInterface|GetMaterialInstance _lockscreenwidget))
  ;[1 1 1] Replication event for when bIsUnlocked changes
  (Class|SquadLockHoloUI|UpdateUnlockState (Variables|Default|GetLockScreenUI) (Variables|Default|GetbIsUnlocked))
  (Rendering|Material|SetVectorParameterValue _returnvalue "Glow Color Inner" (Math|Color|SelectColor (Math|Color|MakeColor 1.0 5.0 1.0) (Math|Color|MakeColor 5.0 3.0 4.0) (Variables|Default|GetbIsUnlocked)) :!class /Script/Engine.MaterialInstanceDynamic)
  (Rendering|Material|SetVectorParameterValue _returnvalue "Glow Color Outer" (Math|Color|SelectColor (Math|Color|MakeColor 15.0 6e+01 24.0) (Math|Color|MakeColor 9e+01 9.5 3.75) (Variables|Default|GetbIsUnlocked)) :!class /Script/Engine.MaterialInstanceDynamic))

(event Custom|StartAllEyeAimScanning
  (bind (_output_pawn_17_ca2911e448c8191bd27fa28c945c96fa _output_smcomp_2_9161e59c4f3629e93c3b238aa11cb77b) (Utilities|Struct|BreakSquadLockScannerRay (Utilities|Array|Get(aref) (Variables|Default|GetScannerRays) 1)))
  (bind (_output_pawn_17_ca2911e448c8191bd27fa28c945c96fa_2 _output_smcomp_2_9161e59c4f3629e93c3b238aa11cb77b_2) (Utilities|Struct|BreakSquadLockScannerRay (Utilities|Array|Get(aref) (Variables|Default|GetScannerRays) 1)))
  (bind _lock (Variables|Default|GetLock))
  ;[0.02054 0 1] Whenever any of the scanners start, start this.
  (Audio|PlaySoundatLocation (Variables|Default|GetScannerOnSound) (Transformation|GetActorLocation))
  (Components|Timeline|Stop (Variables|SquadLock|GetUnAim%20Eye))
  (Utilities|Time|ClearandInvalidateTimerbyHandle (Variables|Default|GetTimer_UnAimEye))
  ;;[1 1 1] Set the volume scanner visible
  (Rendering|SetVisibility _output_smcomp_2_9161e59c4f3629e93c3b238aa11cb77b true false)
  (Rendering|Components|Light|SetLightColor (Variables|Default|GetPointLight) (Variables|Default|GetColorUnlock) true)
  (bind (_direction _eyelerp _stareintent _objectscale _lerpeyecolor _adjusteyerings) (Timeline|Aim%20Eye Play)
    (:Update
      (Rendering|Material|SetCustomPrimitiveDataFloat _lock 12 _eyelerp)
      (Rendering|Material|SetCustomPrimitiveDataFloat _lock 13 _stareintent)
      (Rendering|Material|SetCustomPrimitiveDataFloat _lock 26 (.x _adjusteyerings))
      (Rendering|Material|SetCustomPrimitiveDataFloat _lock 27 (.y _adjusteyerings))
      (Rendering|Material|SetCustomPrimitiveDataVector3 (Variables|Default|GetLock) 18 (Math|Vector|Lerp(Vector) (Math|Conversions|ToVector(LinearColor) (Variables|Default|GetColorLock)) (Math|Conversions|ToVector(LinearColor) (Variables|Default|GetScannerColor)) _lerpeyecolor))
      ;;[0.3802 0.3802 0.3802] Activate Scan volume NS
      (CallFunction|ActivateScanVolumeNS)
      ;;[1 1 1] Show the spotlight
      (Development|SetHiddeninGame (Variables|Default|GetSpotLight) false false)
      (for _array_element (Variables|Default|GetScannerRays)
        (bind (_array_element_pawn_17_ca2911e448c8191bd27fa28c945c96fa _array_element_smcomp_2_9161e59c4f3629e93c3b238aa11cb77b _array_element_aimactive_10_7ce804af4b674e9fe898fcbb9eda2c90) (Utilities|Struct|BreakSquadLockScannerRay _array_element))
        (if _array_element_aimactive_10_7ce804af4b674e9fe898fcbb9eda2c90
          (Transformation|SetRelativeScale3D _array_element_smcomp_2_9161e59c4f3629e93c3b238aa11cb77b (* (Math|Conversions|ToVector(Float) _objectscale) "3.000000,1.000000,1.000000"))))
      (Transformation|SetRelativeScale3D _output_smcomp_2_9161e59c4f3629e93c3b238aa11cb77b_2 (* (Math|Conversions|ToVector(Float) _objectscale) "3.000000,1.000000,1.000000")))))

(event Custom|Activate%20Scan%20Volume%20NS
  (bind _ns_squadlock_scanvolume (Variables|Default|GetNS_SquadLock_ScanVolume))
  ;[0.02054 0 1] Whenever any of the scanners start, start this.
  ;;[0.3802 0.3802 0.3802] Activate Scan volume NS
  (Components|Activation|Activate _ns_squadlock_scanvolume false)
  (Niagara|SetNiagaraVariable(Float) _ns_squadlock_scanvolume "AgeInput" 0.0))

(event Custom|StartScanningAll
  ;[1 1 1] Start/Update Scanning Cues
  (if (not (or (Variables|Default|GetIsScanning) (Variables|Default|GetDisablingScanner)))
    (Variables|Default|SetDisablingScanner false)
    (CallFunction|UpdateScanningQues)))

(event Custom|Update%20Scanning%20Ques
  (bind _quedforunscanfortpawns (Variables|Default|GetQuedForUnScanFortPawns))
  (bind _output (Utilities|Array|Get(acopy) _quedforunscanfortpawns))
  (bind _currentpawninvolumeindex (Variables|Default|GetCurrentPawnInVolumeIndex))
  (bind _pawnsinvolume (Variables|Default|GetPawnsInVolume))
  (bind (_output_pawn_17_ca2911e448c8191bd27fa28c945c96fa _output_niagarasystem_27_9161e59c4f3629e93c3b238aa11cb77b _output_hasbeenscanned_30_b6428df645eeafb560161c979f7736eb _output_timeentered_33_303c236c49de09444cd74d8bb850e2b0) (Utilities|Struct|BreakPawnsInSquadLockVolume (Utilities|Array|Get(aref) _pawnsinvolume _currentpawninvolumeindex)))
  (bind _currentpawninvolumeindex_2 (Variables|Default|GetCurrentPawnInVolumeIndex))
  (bind _pawnsinvolume_2 (Variables|Default|GetPawnsInVolume))
  (bind _output_pawn_17_ca2911e448c8191bd27fa28c945c96fa_2 (Utilities|Struct|BreakPawnsInSquadLockVolume (Utilities|Array|Get(aref) _pawnsinvolume_2 _currentpawninvolumeindex_2)))
  (bind (_output_pawn_17_ca2911e448c8191bd27fa28c945c96fa_3 _output_niagarasystem_27_9161e59c4f3629e93c3b238aa11cb77b_2 _output_hasbeenscanned_30_b6428df645eeafb560161c979f7736eb_2) (Utilities|Struct|BreakPawnsInSquadLockVolume (Utilities|Array|Get(aref) (Variables|Default|GetPawnsInVolume) (Variables|Default|GetCurrentPawnInVolumeIndex))))
  ;[1 1 1] Start/Update Scanning Cues
  (if (CallFunction|IsLocked)
    (CallFunction|UpdatePawnCounts true)
    (CallFunction|DrawDebugtoQuedPawns)
    (Utilities|FlowControl|Sequence
      (:then_0
        ;;[1 1 1] Check for pawns that have exited.  Add them to the Unscan Queue
        (bind (_arrayindex _arrayelement) (Utilities|Array|ReverseforEachLoop (Variables|Default|GetPawnsInVolume))
          (:LoopBody
            (bind (_arrayelement_pawn_17_ca2911e448c8191bd27fa28c945c96fa _arrayelement_niagarasystem_27_9161e59c4f3629e93c3b238aa11cb77b) (Utilities|Struct|BreakPawnsInSquadLockVolume _arrayelement))
            (if (not (>= (Utilities|Array|FindItem (Variables|Default|GetAs%20Fort%20Pawn) _arrayelement_pawn_17_ca2911e448c8191bd27fa28c945c96fa) 0))
              (Utilities|Array|AddUnique (Variables|Default|GetQuedForUnScanFortPawns) _arrayelement_pawn_17_ca2911e448c8191bd27fa28c945c96fa)
              (Utilities|IsValid _arrayelement_niagarasystem_27_9161e59c4f3629e93c3b238aa11cb77b
                (:Is%20Valid
                  ;;;[0.3411 0.5692 1] end particles
                  (Niagara|SetNiagaraVariable(Float) _arrayelement_niagarasystem_27_9161e59c4f3629e93c3b238aa11cb77b "AgeInput" 1.0) ; ideally this would run off a timer
                  (Components|Activation|Deactivate _arrayelement_niagarasystem_27_9161e59c4f3629e93c3b238aa11cb77b))
                (:Is%20Not%20Valid
                  _))
              (Utilities|Array|RemoveIndex (Variables|Default|GetPawnsInVolume) _arrayindex)
              (CallFunction|UpdatePawnCounts false)
              (CallFunction|DrawDebugtoQuedPawns)))))
      (:then_1
        ;;[1 1 1] Add any unscanned pawns to our Scan Queue
        (for _array_element (Variables|Default|GetAs%20Fort%20Pawn)
          (bind _retindex (CallFunction|GetPawnsinVolumeIndexbyPawn _array_element))
          (if (not (>= _retindex 0))
            (Utilities|Array|Add (Variables|Default|GetPawnsInVolume) (Utilities|Struct|MakePawnsInSquadLockVolume _array_element 0 false (Utilities|Time|GetTimeSeconds)))
            (CallFunction|UpdatePawnCounts false)
            (CallFunction|DrawDebugtoQuedPawns))))
      (:then_2
        ;;[1 1 1] Update Unscan Queue
        (if (>= (Utilities|Array|Length _quedforunscanfortpawns) 1)
          (bind (_success _scannerindex) (CallFunction|FindScannerSlotforPawn _output true))
          (if _success
            (CallFunction|CEEndAimEye _scannerindex _output)
            (Utilities|Array|RemoveIndex (Variables|Default|GetQuedForUnScanFortPawns))
            (CallFunction|DrawDebugtoQuedPawns))
          (else
            ;;[1 1 1] Update Scan Queue
            (bind _retindex_2 (CallFunction|GetNextUnscannedPawninVolumeIndex))
            (if (>= _retindex_2 0)
              (Variables|Default|SetCurrentPawnInVolumeIndex _retindex_2)
              (bind (_success_2 _scannerindex_2) (CallFunction|FindScannerSlotforPawn _output_pawn_17_ca2911e448c8191bd27fa28c945c96fa))
              (if _success_2
                (CallFunction|CEAimEye _output_pawn_17_ca2911e448c8191bd27fa28c945c96fa _scannerindex_2)
                (if (not _output_hasbeenscanned_30_b6428df645eeafb560161c979f7736eb_2)
                  (CallFunction|SpawnScannedPawnParticles _output_pawn_17_ca2911e448c8191bd27fa28c945c96fa _output_timeentered_33_303c236c49de09444cd74d8bb850e2b0))
                (CallFunction|DrawDebugtoQuedPawns))
              (elif (or (Utilities|Array|IsEmpty (Variables|Default|GetPawnsInVolume)) (Variables|Default|GetbIsUnlocked))
                ;;[1 1 1] Set a timer to update the ques again later if our ques aren't emptied.  If they are all empty,. turn off the scanner.
                (CallFunction|UnaimEyeAll)
                (else
                  ;;[1 1 1] Cycle through rescanning all the previously scanned pawns
                  (bind _output_get (Variables|Default|SetCurrentPawnInVolumeIndex (Math|Random|RandomInteger (Utilities|Array|Length (Variables|Default|GetPawnsInVolume)))))
                  (if (not (Utilities|Array|IsValidIndex (Variables|Default|GetPawnsInVolume) _output_get))
                    (Variables|Default|SetCurrentPawnInVolumeIndex 0))
                  (bind (_success_3 _scannerindex_3) (CallFunction|FindScannerSlotforPawn _output_pawn_17_ca2911e448c8191bd27fa28c945c96fa_2))
                  (CallFunction|CEAimEye _output_pawn_17_ca2911e448c8191bd27fa28c945c96fa_2 _scannerindex_3))))))))))

(event Custom|SpawnScannedPawnParticles
  (params (InFortPawn :type Fort%20Pawn%20Object%20Reference) (InScannedTime :type Float%20%28double-precision%29%20%28by%20ref%29) (NewParam :type Boolean))
  ;[1 1 1] Start/Update Scanning Cues
  ;;[0.3411 0.5692 1] spawn particles
  (bind _returnvalue (Niagara|SpawnSystemAttached "/SquadLockGameplay/items/SquadLock/Assets/FX/NS_SquadLock_Scanner.NS_SquadLock_Scanner" (Variables|/Script/Engine.Character|GetMesh InFortPawn) "Head" (+ (Transformation|GetWorldLocation (Variables|/Script/Engine.Character|GetMesh InFortPawn)) "0,0,100.000000") "0, 0, 0" "KeepWorldPosition"))
  (Niagara|SetNiagaraVariable(Vector3) _returnvalue "ScannerCenter" (Transformation|GetActorLocation))
  (Utilities|Array|SetArrayElem (Variables|Default|GetPawnsInVolume) (Variables|Default|GetCurrentPawnInVolumeIndex) (Utilities|Struct|MakePawnsInSquadLockVolume InFortPawn _returnvalue true InScannedTime) false)
  (CallFunction|PingIfNeeded)
  (CallFunction|UpdatePawnCounts false))

(event Custom|DrawDebugToQuedPawns
  ;[0.4444 1 0.7184] Debug Lines
  (if (Variables|Default|GetDrawDebuglinesToScannedPawns)
    (for _array_element (Variables|Default|GetPawnsInVolume)
      (bind (_array_element_pawn_17_ca2911e448c8191bd27fa28c945c96fa _array_element_niagarasystem_27_9161e59c4f3629e93c3b238aa11cb77b _array_element_hasbeenscanned_30_b6428df645eeafb560161c979f7736eb _array_element_timeentered_33_303c236c49de09444cd74d8bb850e2b0) (Utilities|Struct|BreakPawnsInSquadLockVolume _array_element))
      (if (>= (- (Utilities|Time|GetTimeSeconds) _array_element_timeentered_33_303c236c49de09444cd74d8bb850e2b0) (Variables|Default|GetSecondsForPendingPawnToBecomeScanned))
        (Rendering|Debug|DrawDebugLine (Transformation|GetActorLocation) (Transformation|GetActorLocation _array_element_pawn_17_ca2911e448c8191bd27fa28c945c96fa) "(R=0.000000,G=1.000000,B=0.051178,A=1.000000)" 0.1 2.0)
        (elif _array_element_hasbeenscanned_30_b6428df645eeafb560161c979f7736eb
          (Rendering|Debug|DrawDebugLine (Transformation|GetActorLocation) (Transformation|GetActorLocation _array_element_pawn_17_ca2911e448c8191bd27fa28c945c96fa) "(R=1.000000,G=0.993446,B=0.000000,A=1.000000)" 0.1 5.0)
          (else
            (Rendering|Debug|DrawDebugLine (Transformation|GetActorLocation) (Transformation|GetActorLocation _array_element_pawn_17_ca2911e448c8191bd27fa28c945c96fa) "(R=0.531250,G=0.425751,B=0.000000,A=1.000000)" 0.1 2.0)))))
    (for _array_element_2 (Variables|Default|GetQuedForUnScanFortPawns)
      (Rendering|Debug|DrawDebugLine (Transformation|GetActorLocation) (Transformation|GetActorLocation _array_element_2) "(R=1.000000,G=0.000000,B=0.034335,A=1.000000)" 0.1 5.0))))

(event Custom|Start%20Scanner%20Aim%20Interpolation
  ;[1 1 1] Start and update the scanner ray aim interpolation
  (Utilities|Time|SetTimerforNextTickbyEvent (delegate Custom|FrameUpdateAimInterpolation)))

(event Custom|SetupAndStartScannerAimInterpolation
  (params (InAimRot :type Rotator) (LinearColor :type Linear%20Color%20Structure) (IsExit :type Boolean))
  (bind _x (Math|Rotator|GetAxes InAimRot))
  (bind _scannerrayanglelimitfactor_mindot (Variables|ScannerRays|GetScannerRayAngleLimitFactor_MinDot))
  ;[1 1 1] SetDesiredAimRotation -  called once to setup and start lerping the ray.  This limits the angle if needed)
  (Utilities|FlowControl|Sequence
    (:then_0
      (Utilities|FlowControl|Sequence
        (:then_0
          (bind _newparam (CallFunction|GetScannerForwardRot))
          (if (< (Math|Vector|DotProduct _x (Math|Rotator|GetAxes _newparam)) _scannerrayanglelimitfactor_mindot)
            (Variables|ScannerRays|SetPrimaryDesiredRot (Math|Conversions|RotationFromXVector (+ (* (Math|Rotator|GetAxes (Math|Rotator|MakeRotfromXY (Math|Rotator|GetAxes _newparam) _x)) (Math|Trig|Cos(Degrees) (Math|Trig|Acos(Degrees) _scannerrayanglelimitfactor_mindot))) (* (.Y (Math|Rotator|GetAxes (Math|Rotator|MakeRotfromXY (Math|Rotator|GetAxes _newparam) _x))) (Math|Trig|Sin(Degrees) (Math|Trig|Acos(Degrees) _scannerrayanglelimitfactor_mindot))))))
            (else
              (Variables|ScannerRays|SetPrimaryDesiredRot InAimRot)
              (Variables|ScannerRays|SetStartingtRot (Variables|ScannerRays|GetCurrentRot))
              (CallFunction|SetPrimaryAimtoCurrent)
              (Variables|ScannerRays|SetAimInterpolationAlpha 0.0))))
        (:then_1
          (Variables|ScannerRays|SetDesiredColor LinearColor)
          (Variables|ScannerRays|SetStartColor (Variables|ScannerRays|GetCurrentColor)))
        (:then_2
          (CallFunction|StartScannerAimInterpolation))))
    (:then_1)))

(event Custom|FrameUpdateAimInterpolation
  ;[1 1 1] Start and update the scanner ray aim interpolation
  (bind _output_get (Variables|ScannerRays|SetAimInterpolationAlpha (Math|Float|Clamp(Float) (+ (Variables|ScannerRays|GetAimInterpolationAlpha) (* (Utilities|Time|GetWorldDeltaSeconds) 3.5)))))
  (Variables|ScannerRays|SetCurrentRot (Math|Rotator|Lerp(Rotator) (Variables|ScannerRays|GetStartingtRot) (Variables|ScannerRays|GetPrimaryDesiredRot) _output_get true))
  (bind _output_get_2 (Variables|ScannerRays|SetCurrentColor (Math|Color|Lerp(LinearColor) (Variables|ScannerRays|GetStartColor) (Variables|ScannerRays|GetDesiredColor) _output_get)))
  (CallFunction|SetScannerColor (Variables|Default|GetSm_Camera_ScannerRay_ForPawns) _output_get_2)
  (CallFunction|SetPrimaryAimtoCurrent)
  (if (not (>= _output_get 1.0))
    (Utilities|Time|SetTimerforNextTickbyEvent (delegate Custom|FrameUpdateAimInterpolation))))

(event Custom|SetPrimaryAimToCurrent
  (bind (_output_pawn_17_ca2911e448c8191bd27fa28c945c96fa _output_smcomp_2_9161e59c4f3629e93c3b238aa11cb77b) (Utilities|Struct|BreakSquadLockScannerRay (Utilities|Array|Get(aref) (Variables|Default|GetScannerRays))))
  (bind _currentrot (Variables|ScannerRays|GetCurrentRot))
  ;[1 1 1] Start and update the scanner ray aim interpolation
  ;;[1 1 1] SetPrimayRayToTheCurrentRotation
  (Transformation|SetWorldRotation _output_smcomp_2_9161e59c4f3629e93c3b238aa11cb77b _currentrot false false)
  (Rendering|Material|SetCustomPrimitiveDataVector3 (Variables|Default|GetLock) 8 (Math|Vector|MakeVector (.x (+ (Transformation|GetActorLocation) (* (Math|Vector|GetForwardVector _currentrot) 5e+02))) (.y (+ (Transformation|GetActorLocation) (* (Math|Vector|GetForwardVector _currentrot) 5e+02))) (.z (+ (Transformation|GetActorLocation) (* (Math|Vector|GetForwardVector _currentrot) 5e+02)))))
  (Rendering|Components|Light|SetLightColor (Variables|Default|GetSpotLight) (Variables|ScannerRays|GetCurrentColor) true))

(event Custom|SetScannerPower
  (params (On :type Boolean))
  (bind _scannerrays (Variables|Default|GetScannerRays))
  (bind (_output_pawn_17_ca2911e448c8191bd27fa28c945c96fa _output_smcomp_2_9161e59c4f3629e93c3b238aa11cb77b) (Utilities|Struct|BreakSquadLockScannerRay (Utilities|Array|Get(aref) _scannerrays 1)))
  (bind _scannerrays_2 (Variables|Default|GetScannerRays))
  (bind (_output_pawn_17_ca2911e448c8191bd27fa28c945c96fa_2 _output_smcomp_2_9161e59c4f3629e93c3b238aa11cb77b_2) (Utilities|Struct|BreakSquadLockScannerRay (Utilities|Array|Get(aref) _scannerrays_2)))
  (bind _spotlight (Variables|Default|GetSpotLight))
  ;[1 1 1] Set Scanner Power
  (Class|SquadLockHoloUI|SetMeepleVisibility (Variables|Default|GetLockScreenUI) On)
  (if On
    (Development|SetHiddeninGame _spotlight false false)
    (Rendering|SetVisibility _output_smcomp_2_9161e59c4f3629e93c3b238aa11cb77b_2 true false)
    (Rendering|SetVisibility _output_smcomp_2_9161e59c4f3629e93c3b238aa11cb77b true false)
    (Rendering|Components|Light|SetLightColor (Variables|Default|GetPointLight) (Variables|Default|GetColorUnlock) true)
    (CallFunction|ActivateScanVolumeNS)
    (bind (_arrayindex _arrayelement) (Utilities|Array|ReverseforEachLoop (Variables|Default|GetPawnsInVolume))
      (:LoopBody
        (Variables|Default|SetCurrentPawnInVolumeIndex _arrayindex)
        (bind (_arrayelement_pawn_17_ca2911e448c8191bd27fa28c945c96fa _arrayelement_niagarasystem_27_9161e59c4f3629e93c3b238aa11cb77b _arrayelement_hasbeenscanned_30_b6428df645eeafb560161c979f7736eb _arrayelement_timeentered_33_303c236c49de09444cd74d8bb850e2b0) (Utilities|Struct|BreakPawnsInSquadLockVolume _arrayelement))
        (CallFunction|SpawnScannedPawnParticles _arrayelement_pawn_17_ca2911e448c8191bd27fa28c945c96fa _arrayelement_timeentered_33_303c236c49de09444cd74d8bb850e2b0)))
    (else
      (Development|SetHiddeninGame _spotlight true false)
      (Rendering|SetVisibility _output_smcomp_2_9161e59c4f3629e93c3b238aa11cb77b_2 false false)
      (Rendering|SetVisibility _output_smcomp_2_9161e59c4f3629e93c3b238aa11cb77b false false)
      (Rendering|Components|Light|SetLightColor :self (Variables|Default|GetPointLight) :bSRGB true)
      (CallFunction|DeactivateScanVolumeNS)
      (bind (_arrayindex_2 _arrayelement_2) (Utilities|Array|ReverseforEachLoop (Variables|Default|GetPawnsInVolume))
        (:LoopBody
          (bind (_arrayelement_pawn_17_ca2911e448c8191bd27fa28c945c96fa_2 _arrayelement_niagarasystem_27_9161e59c4f3629e93c3b238aa11cb77b_2) (Utilities|Struct|BreakPawnsInSquadLockVolume _arrayelement_2))
          (Niagara|SetNiagaraVariable(Float) _arrayelement_niagarasystem_27_9161e59c4f3629e93c3b238aa11cb77b_2 "AgeInput" 1.0) ; ideally this would run off a timer
          (Components|Activation|Deactivate _arrayelement_niagarasystem_27_9161e59c4f3629e93c3b238aa11cb77b_2))))))

(fn BlueprintCanInteract
  (params (InteractingPawn :type Fort%20Pawn%20Object%20Reference) (InteractionBeingAttempted :type EInteractionBeingAttempted%20Enum :default "FirstInteraction") (InteractionType :type TInteractionType%20Enum :default "IT_NoInteraction"))
  (return false))

(fn BlueprintGetFailedInteractionString
  (params (InteractingPawn :type Fort%20Pawn%20Object%20Reference) (InteractionBeingAttempted :type EInteractionBeingAttempted%20Enum :default "FirstInteraction"))
  (return ""))

(fn BlueprintGetInteractionString
  (params (InteractingPawn :type Fort%20Pawn%20Object%20Reference) (InteractionBeingAttempted :type EInteractionBeingAttempted%20Enum :default "FirstInteraction"))
  (return ""))

(fn CanDisplayEasyInteractButton
  (return false))

(fn CountActiveScanners
  (LocalVariables|SetlocalActiveCount 0)
  (for (_ _array_index) (Variables|Default|GetScannerRays)
    (bind (_output_pawn_17_ca2911e448c8191bd27fa28c945c96fa _output_smcomp_2_9161e59c4f3629e93c3b238aa11cb77b _output_aimactive_10_7ce804af4b674e9fe898fcbb9eda2c90) (Utilities|Struct|BreakSquadLockScannerRay (Utilities|Array|Get(aref) (Variables|Default|GetScannerRays) _array_index)))
    (if _output_aimactive_10_7ce804af4b674e9fe898fcbb9eda2c90
      (Math|Integer|IncrementInt (LocalVariables|GetlocalActiveCount))))
  (Variables|Default|SetActiveScannerCount (LocalVariables|GetlocalActiveCount)))

(fn EndScan
  (params (Canceled :type Boolean)))

(fn FindScannerSlotForPawn
  (params (InFortPawn :type Fort%20Pawn%20Object%20Reference) (Exiting :type Boolean))
  (bind (_output_pawn_17_ca2911e448c8191bd27fa28c945c96fa _output_smcomp_2_9161e59c4f3629e93c3b238aa11cb77b _output_aimactive_10_7ce804af4b674e9fe898fcbb9eda2c90 _output_unaimactive_12_9b66154e49cf5dd3a1cabab49f3169ac) (Utilities|Struct|BreakSquadLockScannerRay (Utilities|Array|Get(aref) (Variables|Default|GetScannerRays))))
  (if Exiting
    (LocalVariables|SetDebugLineColor "(R=1.000000,G=0.000000,B=0.000000,A=1.000000)")
    (else
      (LocalVariables|SetDebugLineColor "(R=0.000000,G=1.000000,B=0.245718,A=1.000000)")))
  (LocalVariables|SetRetValScannerIndex -1)
  (Variables|Default|SetFoundScannerForPawn false)
  ;[1 1 1] See if we can find an empty slot for this pawn
  (if (and (not _output_aimactive_10_7ce804af4b674e9fe898fcbb9eda2c90) (not _output_unaimactive_12_9b66154e49cf5dd3a1cabab49f3169ac))
    (Variables|Default|SetFoundScannerForPawn true)
    (bind _output_get (LocalVariables|SetRetValScannerIndex 0))
    (return true _output_get)
    (else
      ;[1 1 1] If we couldn't find a slot break out outer loop since no other pawns will find slots either
      (return (Variables|Default|GetFoundScannerForPawn) (LocalVariables|GetRetValScannerIndex)))))

(fn GetNextUnscannedPawnInVolumeIndex
  (LocalVariables|SetCurrentTime (Utilities|Time|GetTimeSeconds))
  (LocalVariables|SetUnscannedPawnCount 0)
  ;[1 1 1] If we have any that have never been scanned return that index
  (for (_array_element _array_index) (Variables|Default|GetPawnsInVolume)
    (bind (_array_element_pawn_17_ca2911e448c8191bd27fa28c945c96fa _array_element_niagarasystem_27_9161e59c4f3629e93c3b238aa11cb77b _array_element_hasbeenscanned_30_b6428df645eeafb560161c979f7736eb _array_element_timeentered_33_303c236c49de09444cd74d8bb850e2b0) (Utilities|Struct|BreakPawnsInSquadLockVolume _array_element))
    (if (== _array_element_hasbeenscanned_30_b6428df645eeafb560161c979f7736eb false)
      (return _array_index)
      (elif (>= (- (LocalVariables|GetCurrentTime) _array_element_timeentered_33_303c236c49de09444cd74d8bb850e2b0) (Variables|Default|GetSecondsForPendingPawnToBecomeScanned))
        _
        (else
          (Math|Integer|IncrementInt (LocalVariables|GetUnscannedPawnCount))))))
  ;[1 1 1] If there are any stil being scanned return one of them randomly
  (if (>= (LocalVariables|GetUnscannedPawnCount) 1)
    (LocalVariables|SetRandomOfUnscannedPawnCount (Math|Random|RandomInteger (LocalVariables|GetUnscannedPawnCount)))
    (LocalVariables|SetUnscannedPawnCount 0)
    (for (_array_element_2 _array_index_2) (Variables|Default|GetPawnsInVolume)
      (bind (_array_element_pawn_17_ca2911e448c8191bd27fa28c945c96fa_2 _array_element_niagarasystem_27_9161e59c4f3629e93c3b238aa11cb77b_2 _array_element_hasbeenscanned_30_b6428df645eeafb560161c979f7736eb_2 _array_element_timeentered_33_303c236c49de09444cd74d8bb850e2b0_2) (Utilities|Struct|BreakPawnsInSquadLockVolume _array_element_2))
      (if (== _array_element_hasbeenscanned_30_b6428df645eeafb560161c979f7736eb_2 false)
        _
        (elif (>= (- (LocalVariables|GetCurrentTime) _array_element_timeentered_33_303c236c49de09444cd74d8bb850e2b0_2) (Variables|Default|GetSecondsForPendingPawnToBecomeScanned))
          _
          (elif (== (LocalVariables|GetUnscannedPawnCount) (LocalVariables|GetRandomOfUnscannedPawnCount))
            (return _array_index_2)
            (else
              (Math|Integer|IncrementInt (LocalVariables|GetUnscannedPawnCount)))))))
    (return -1)
    (else
      (return -1))))

(fn GetNumRequiredPlayers
  (bind _adjustednumplayersrequiredtounlock (Variables|Default|GetAdjustedNumPlayersRequiredToUnlock))
  (if (> _adjustednumplayersrequiredtounlock 0)
    (return _adjustednumplayersrequiredtounlock)
    (else
      (Development|PrintString "ERROR: Adjusted number of players required has not been replicated.  Cannot return valid number here.")
      (return 2))))

(fn GetPawnsInVolumeIndexByPawn
  (params (FortPawn :type Fort%20Pawn%20Object%20Reference))
  (for (_array_element _array_index) (Variables|Default|GetPawnsInVolume)
    (bind _array_element_pawn_17_ca2911e448c8191bd27fa28c945c96fa (Utilities|Struct|BreakPawnsInSquadLockVolume _array_element))
    (if (== _array_element_pawn_17_ca2911e448c8191bd27fa28c945c96fa FortPawn)
      (return _array_index)))
  (return -1))

(fn GetPlayersInRangeandTotalRequiredToUnlock
  ;[1 1 1] All players can unlock (regardless of squad)
  (bind _getnumrequiredplayers (CallFunction|GetNumRequiredPlayers))
  (bind (_retval_scannedpawncount _retval_pendingscannedpawncount) (CallFunction|GetPresentPawnCounts (Utilities|Time|GetTimeSeconds)))
  (return _retval_scannedpawncount _getnumrequiredplayers _retval_pendingscannedpawncount))

(fn GetPresentPawnCounts
  (params (CurrentTime :type Float%20%28double-precision%29))
  (LocalVariables|SetLocal_PendingPawnCount 0)
  (LocalVariables|SetLocal_ScannedPawnCount 0)
  (for _array_element (Variables|Default|GetPawnsInVolume)
    (bind (_array_element_pawn_17_ca2911e448c8191bd27fa28c945c96fa _array_element_niagarasystem_27_9161e59c4f3629e93c3b238aa11cb77b _array_element_hasbeenscanned_30_b6428df645eeafb560161c979f7736eb _array_element_timeentered_33_303c236c49de09444cd74d8bb850e2b0) (Utilities|Struct|BreakPawnsInSquadLockVolume _array_element))
    (if (>= (- CurrentTime _array_element_timeentered_33_303c236c49de09444cd74d8bb850e2b0) (Variables|Default|GetSecondsForPendingPawnToBecomeScanned))
      (Math|Integer|IncrementInt (LocalVariables|GetLocal_ScannedPawnCount))
      (else
        (Math|Integer|IncrementInt (LocalVariables|GetLocal_PendingPawnCount)))))
  (return (LocalVariables|GetLocal_ScannedPawnCount) (LocalVariables|GetLocal_PendingPawnCount)))

(fn GetScannerForwardRot
  (return (Math|Rotator|MakeRotfromX (* (Math|Vector|GetRightVector (Transformation|GetActorRotation)) -1.0))))

(fn GetMarkerDetailsId
  (return "MarkerDetails.SquadLock"))

(fn InitLockScreenWidget
  (bind _lockscreenwidget (Variables|Default|GetLockScreenWidget))
  (bind _returnvalue (UserInterface|GetUserWidgetObject _lockscreenwidget))
  (Networking|SwitchHasAuthority
    (:Remote
      (Utilities|FlowControl|Sequence
        (:then_0
          (bind _assquad_lock_holo_ui (Utilities|Casting|CastToSquadLock_HoloUI _returnvalue)
            (:then
              (Variables|Default|SetLockScreenUI _assquad_lock_holo_ui)
              (bind _getnumrequiredplayers (CallFunction|GetNumRequiredPlayers))
              (Class|SquadLockHoloUI|UpdateMaxPlayerCount (Variables|Default|GetLockScreenUI) _getnumrequiredplayers)
              (Class|SquadLockHoloUI|UpdateCurrentPlayerCount (Variables|Default|GetLockScreenUI) (Variables|Default|GetNumPlayersInRange) (Variables|Default|GetNumPendingPlayersInRange)))
            (:CastFailed
              (Utilities|FlowControl|DoOnce false
                (:Completed
                  (Fortnite|FortShippingLog 0 (Utilities|String|Append "SQUADLOCKS: Failed to cast Squad Lock Holo UI from: " (Utilities|GetDisplayName _returnvalue))))))))
        (:then_1
          (CallFunction|OnRepBIsUnlocked))))))

(fn InitializeScannerRay
  (params (ScannerRayStaticMesh :type Static%20Mesh%20Component%20Object%20Reference) (ScannerColor :type Linear%20Color%20Structure) (baseWidth :type Float%20%28double-precision%29) (endWidth :type Float%20%28double-precision%29) (Angle :type Float%20%28double-precision%29 :default 12.0) (ScanRotAnim :type Float%20%28double-precision%29 :default 1.0))
  (Rendering|Material|SetCustomPrimitiveDataFloat ScannerRayStaticMesh 2 baseWidth) ; WidthBase
  (Rendering|Material|SetCustomPrimitiveDataFloat ScannerRayStaticMesh 1 endWidth) ; WidthEnd
  (Rendering|Material|SetCustomPrimitiveDataFloat ScannerRayStaticMesh 12 1.0) ; SecondaryMovementSpeed
  (Rendering|Material|SetCustomPrimitiveDataFloat ScannerRayStaticMesh 13 0.5) ; SecondaryFanOpacity
  (Rendering|Material|SetCustomPrimitiveDataFloat ScannerRayStaticMesh 0 Angle) ; Angle
  (Rendering|Material|SetCustomPrimitiveDataFloat ScannerRayStaticMesh 3 0.25) ; FalloffHardness
  (CallFunction|SetScannerColor ScannerRayStaticMesh ScannerColor)
  (Rendering|Material|SetScalarParameterValueonMaterials ScannerRayStaticMesh "ScanRotAnim" ScanRotAnim))

(fn IsLocked
  (return (not (or (Variables|Default|GetPendingUnlock) (Utilities|Enum|NotEqual(Enum) (Variables|LockParent|GetLockState) "NewEnumerator1")))))

(fn OnRep_AdjustedNumPlayersRequiredToUnlock
  (Networking|SwitchHasAuthority
    (:Remote
      (CallFunction|InitLockScreenWidget)
      (CallFunction|OnRepBIsUnlocked))))

(fn OnRep_AsFortPawn
  (CallFunction|StartScanningAll))

(fn OnRep_EndType
  ;[1 1 1] Rep notify the end with the end type. This proves to work better than having the end scan be a multicast.
  (switch int (Variables|Default|GetEndType)
    (:0
      (CallFunction|EndScan true))
    (:1
      (CallFunction|EndScan false))))

(fn OnRep_HasBeenDistabled
  (Rendering|Material|SetMaterial (Variables|Default|GetLock) 0 "/SquadLockGameplay/items/SquadLock/Assets/Materials/MI_Squad_Lock_Black.MI_Squad_Lock_Black")
  (Rendering|SetVisibility (Variables|Default|GetLockScreenWidget) false true))

(fn OnRep_NumPendingPlayersInRange
  (bind _numpendingplayersinrange (Variables|Default|GetNumPendingPlayersInRange))
  (bind _numplayersinrange (Variables|Default|GetNumPlayersInRange))
  (bind _returnvalue (+ _numplayersinrange _numpendingplayersinrange))
  (bind _numpendingplayersinrange_2 (Variables|Default|GetNumPendingPlayersInRange))
  ;[0.5724 1 0.9801] Debug
  (Development|PrintString (Utilities|String|Append "OnRep NumPendingPlayersInRange: " (Utilities|String|ToString(Integer) (Variables|Default|GetNumPendingPlayersInRange)) "(was " (Utilities|String|ToString(Integer) (Variables|Default|GetPrevNumPendingPlayersInRange)) ")") (Variables|Default|GetDrawDebuglinesToScannedPawns) (Variables|Default|GetDrawDebuglinesToScannedPawns))
  (Class|SquadLockHoloUI|UpdateCurrentPlayerCount (Variables|Default|GetLockScreenUI) (Variables|Default|GetNumPlayersInRange) _numpendingplayersinrange_2)
  (Utilities|FlowControl|Sequence
    (:then_0
      (if (< (Variables|Default|GetPrevNumPendingPlayersInRange) (Variables|Default|GetNumPendingPlayersInRange))
        (Utilities|FlowControl|Sequence
          (:then_0
            (if (not (Variables|Default|GetbIsUnlocked))
              (bind _audio (Variables|Default|GetAudio)
                (:then
                  (Audio|SpawnSoundAttached (Variables|Default|GetEnterRangeSound) _audio))
                (:else))))
          (:then_1
            (if (> _returnvalue 0)
              (bind _audio_2 (Variables|Default|GetAudio)
                (:then
                  (Audio|Components|Audio|FadeIn _audio_2 0.3 1.0 0.0 "Linear" :!class /Script/Engine.AudioComponent))
                (:else)))))
        (else
          (Utilities|FlowControl|Sequence
            (:then_0
              (if (< _returnvalue 1)
                (bind _audio_3 (Variables|Default|GetAudio)
                  (:then
                    (Audio|Components|Audio|FadeOut _audio_3 0.1 0.0 "Linear" :!class /Script/Engine.AudioComponent))
                  (:else))))))))
    (:then_1
      (Variables|Default|SetPrevNumPendingPlayersInRange _numpendingplayersinrange_2))))

(fn OnRep_NumPlayersInRange
  (bind _numpendingplayersinrange (Variables|Default|GetNumPendingPlayersInRange))
  (bind _numplayersinrange (Variables|Default|GetNumPlayersInRange))
  (bind _returnvalue (+ _numplayersinrange _numpendingplayersinrange))
  (bind _numplayersinrange_2 (Variables|Default|GetNumPlayersInRange))
  (Development|PrintString (Utilities|String|Append "OnRep NumPlayersInRange: " (Utilities|String|ToString(Integer) (Variables|Default|GetNumPlayersInRange)) "(was " (Utilities|String|ToString(Integer) (Variables|Default|GetPrevNumPlayersInRange)) ")") (Variables|Default|GetDrawDebuglinesToScannedPawns) (Variables|Default|GetDrawDebuglinesToScannedPawns))
  (Class|SquadLockHoloUI|UpdateCurrentPlayerCount (Variables|Default|GetLockScreenUI) _numplayersinrange_2 (Variables|Default|GetNumPendingPlayersInRange))
  (Utilities|FlowControl|Sequence
    (:then_0
      (if (< (Variables|Default|GetPrevNumPlayersInRange) (Variables|Default|GetNumPlayersInRange))
        (bind _audio (Variables|Default|GetAudio)
          (:then
            (bind _getnumrequiredplayers (CallFunction|GetNumRequiredPlayers))
            (if (not (>= (Variables|Default|GetNumPlayersInRange) _getnumrequiredplayers))
              (Audio|SpawnSoundAttached (Variables|Default|GetRegisteredPlayerSound) _audio))
            (Utilities|FlowControl|Sequence
              (:then_1
                (if (> _returnvalue 0)
                  (bind _audio_2 (Variables|Default|GetAudio)
                    (:then
                      (Audio|Components|Audio|FadeIn _audio_2 0.4 1.0 0.0 "Linear" :!class /Script/Engine.AudioComponent))
                    (:else))))))
          (:else))
        (else
          (Utilities|FlowControl|Sequence
            (:then_0
              (if (< _returnvalue 1)
                (Audio|PlaySoundatLocation (Variables|Default|GetScannerOffSound) (Transformation|GetWorldLocation (Variables|Default|GetAudio)))
                (bind _audio_3 (Variables|Default|GetAudio)
                  (:then
                    (Audio|Components|Audio|FadeOut _audio_3 0.1 0.0 "Linear" :!class /Script/Engine.AudioComponent))
                  (:else))))
            (:then_1
              (if (not (Variables|Default|GetbIsUnlocked))
                (bind _audio_4 (Variables|Default|GetAudio)
                  (:then
                    (Audio|SpawnSoundAttached (Variables|Default|GetExitRangeSound) _audio_4))
                  (:else))))))))
    (:then_1
      (Variables|Default|SetPrevNumPlayersInRange _numplayersinrange_2))))

(fn OnRep_PendingUnlock
  (CallFunction|StartUnlock))

(fn PingifNeeded
  (if (not (Variables|Default|GetHasPinged))
    (Variables|Default|SetHasPinged true)
    (for _array_element (Variables|Default|GetPawnsInVolume)
      (bind _array_element_pawn_17_ca2911e448c8191bd27fa28c945c96fa (Utilities|Struct|BreakPawnsInSquadLockVolume _array_element))
      (if (Pawn|IsLocallyControlled _array_element_pawn_17_ca2911e448c8191bd27fa28c945c96fa)
        (bind _asfort_player_controller_athena (Utilities|Casting|CastToFortPlayerControllerAthena (Pawn|GetController _array_element_pawn_17_ca2911e448c8191bd27fa28c945c96fa))
          (:then
            (Fortnite|RegisterSoundIndicatorPlacementContext _asfort_player_controller_athena self))
          (:CastFailed
            (Utilities|FlowControl|DoOnce false
              (:Completed
                (Fortnite|FortShippingLog 0 (Utilities|String|Append "SQUADLOCKS: Ping if needed failed to cast fort Player Controller from:" (Utilities|GetDisplayName (Pawn|GetController _array_element_pawn_17_ca2911e448c8191bd27fa28c945c96fa))))))))))))

(fn SetScannerColor
  (params (ScannerRayStaticMesh :type Static%20Mesh%20Component%20Object%20Reference) (InScannerColor_R :type Float%20%28single-precision%29) (InScannerColor_G :type Float%20%28single-precision%29) (InScannerColor_B :type Float%20%28single-precision%29) (InScannerColor_A :type Float%20%28single-precision%29))
  (Rendering|Material|SetCustomPrimitiveDataVector4 ScannerRayStaticMesh 4 (Math|Vector4|MakeVector4 InScannerColor_R InScannerColor_G InScannerColor_B InScannerColor_A)) ; Color
  (Rendering|Material|SetCustomPrimitiveDataVector4 ScannerRayStaticMesh 8 (Math|Vector4|MakeVector4 InScannerColor_R InScannerColor_G InScannerColor_B InScannerColor_A)) ; Color
)

(fn SetScanningActive
  (params (Active :type Boolean))
  (bind _fortsoundindicator (Variables|Default|GetFortSoundIndicator))
  (if Active
    (if (not (Variables|Default|GetIsScanning))
      (Audio|Components|Audio|FadeIn (Variables|Default|GetScanSound) 1.2 1.0 0.0 "Linear" :!class /Script/Engine.AudioComponent)
      (Variables|Default|SetIsScanning true)
      (CallFunction|StartAllEyeAimScanning)
      (SoundIndicator|StartSoundIndicatorLooping _fortsoundindicator (Transformation|GetActorLocation) (Variables|/Script/Engine.AudioComponent|GetSound (Variables|Default|GetScanSound)) 1.0 0))
    (elif (Variables|Default|GetIsScanning)
      (Audio|Components|Audio|FadeOut (Variables|Default|GetScanSound) 0.2 0.0 "Linear" :!class /Script/Engine.AudioComponent)
      (Variables|Default|SetIsScanning false)
      (Rendering|Material|SetCustomPrimitiveDataVector3 (Variables|Default|GetLock) 14 (Variables|Default|GetLock_Mat_Direction))
      (SoundIndicator|StopSoundIndicatorLooping _fortsoundindicator))))

(fn UpdatePawnCounts
  (params (IsForTimerUpdate :type Boolean))
  (if (Networking|IsServer)
    (bind (_numplayersnrange _numplayersrequired _numunscannedplayersinrange) (CallFunction|GetPlayersinRangeAndTotalRequiredtoUnlock))
    ;[0.6892 1 0.9297] Debug
    (Development|PrintString (Utilities|String|Append "PawnCounts: " "pending:" (Utilities|String|ToString(Integer) _numunscannedplayersinrange) "    scanned: " (Utilities|String|ToString(Integer) _numplayersnrange) "     required:" (Utilities|String|ToString(Integer) _numplayersrequired)) (Variables|Default|GetDrawDebuglinesToScannedPawns) (Variables|Default|GetDrawDebuglinesToScannedPawns))
    (if (!= (Variables|Default|GetNumPendingPlayersInRange) _numunscannedplayersinrange)
      (Variables|Default|SetNumPendingPlayersInRange _numunscannedplayersinrange))
    (if (!= (Variables|Default|GetNumPlayersInRange) _numplayersnrange)
      (Variables|Default|SetNumPlayersInRange _numplayersnrange))
    (if (and (> _numplayersnrange 0) (>= _numplayersnrange _numplayersrequired))
      ;[0 1 0.3233] Everyone arrived.  Unlock!!
      (Variables|Default|SetPendingUnlock true)
      (else
        ;[1 0.2714 0.1214] Not ready to unlock
        (Variables|Default|SetAllPawnsWithinRange false)))))

(fn ConstructionScript
  (bind _lock (Variables|Default|GetLock))
  (bind _lock_2 (Variables|Default|GetLock))
  (|CallParentFunction)
  (CallFunction|InitializeScannerRay (Variables|Default|GetSm_Camera_ScannerRay_ForPawns) (Variables|ScannerRays|GetPawnEnterScannerColor) (Variables|ScannerRays|GetPawnScannerWidthBase) (Variables|ScannerRays|GetPawnScannerWidthEnd) (Variables|ScannerRays|GetPawnScannerAngle))
  (CallFunction|InitializeScannerRay (Variables|Default|GetSm_Camera_ScannerRay_Volume) (Variables|Default|GetScannerColor) (Variables|ScannerRays|GetVolumeScannerWidthBase) (Variables|ScannerRays|GetVolumeScannerWidthEnd) (Variables|ScannerRays|GetVolumeScannerAngle) 2.0)
  (CallFunction|InitializeScannerRay (Variables|Default|GetSm_Camera_ScannerRay3) (Variables|Default|GetScannerColor) (Variables|ScannerRays|GetPawnScannerWidthBase) (Variables|ScannerRays|GetPawnScannerWidthEnd) (Variables|ScannerRays|GetPawnScannerAngle))
  (CallFunction|InitializeScannerRay (Variables|Default|GetSm_Camera_ScannerRay4) (Variables|Default|GetScannerColor) (Variables|ScannerRays|GetPawnScannerWidthBase) (Variables|ScannerRays|GetPawnScannerWidthEnd) (Variables|ScannerRays|GetPawnScannerAngle))
  ;[1 1 1] Set up the lock material visuals
  (Rendering|Material|SetCustomPrimitiveDataFloat (Variables|Default|GetLock) 12 0.0)
  (bind _output_get (Variables|Default|SetLock_Mat_Direction (+ (Transformation|GetWorldLocation (Variables|Default|GetLock)) (* (Transformation|GetForwardVector (Variables|Default|GetLock)) 4.0))))
  (Rendering|Material|SetCustomPrimitiveDataVector3 _lock_2 14 _output_get)
  (Rendering|Material|SetCustomPrimitiveDataVector3 _lock_2 18 (Math|Conversions|ToVector(LinearColor) (Variables|Default|GetColorLock)))
  (Rendering|Material|SetCustomPrimitiveDataVector3 _lock_2 22 (Math|Conversions|ToVector(LinearColor) (Variables|Default|GetColorUnlock)))
  (Rendering|Components|Light|SetLightColor (Variables|Default|GetPointLight) (Variables|Default|GetColorLock) true)
  (Rendering|Material|SetCustomPrimitiveDataFloat _lock 26 5.0)
  (Rendering|Material|SetCustomPrimitiveDataFloat _lock 27 3.0))
